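# Background block for top nav bar tab buttons in the UIKit modular button system, derived from WBP_UIKit_BlockBackground_Base.
# Transition events animate the FocusOutline material on hover and show the Backing when selected, toggling a post-buffer capture widget alongside it.
# Selection waits briefly, rechecks state, then plays a padding animation and, if UseBackgroundVisuals is set, a flare animation.
# Setters apply materials and configurable backing corner radius; the outline radius is hardcoded.
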WidgetBlueprint /FortUILibrary/UIKit/Buttons/TabButtons/_Blocks/WBP_UIKit_Block_BackgroundTopNavBar
animations [
 Anim_OutlineHover {
  binding_count=1
  length_seconds=1
  object=&self:Anim_OutlineHover
  sequence {
   bindings [FocusOutline {
     id=#1
     possessed_class=Image
     tracks [
      {
       type=MovieSceneFloatTrack
       display_name=Render Opacity
       sections [{type=MovieSceneFloatSection, channels [None {type=MovieSceneScriptingFloatChannel, keys [[0,0],[3,1]]}]}]
      }
      {
       type=MovieSceneWidgetMaterialTrack
       sections [{
         type=MovieSceneParameterSection
         channels [
          Anim_Hover_Fill {type=MovieSceneScriptingFloatChannel, keys [[0,0],[60,1]]}
          Anim_Hover_Energy {type=MovieSceneScriptingFloatChannel, keys [[0,0],[45,1]]}
          Anim_Hover_Outline {type=MovieSceneScriptingFloatChannel, keys [[0,0],[15,1]]}
          Anim_OutlineThickness {type=MovieSceneScriptingFloatChannel, keys [[0,0],[15,1]]}
         ]
        }]
      }
     ]
    }]
   display_rate {numerator=60}
  }
 }
 Anim_OutlineUnhover {
  binding_count=1
  length_seconds=0.5
  object=&self:Anim_OutlineUnhover
  sequence {bindings [FocusOutline {
     id=#2
     possessed_class=Image
     tracks [
      {
       type=MovieSceneFloatTrack
       display_name=Render Opacity
       sections [{type=MovieSceneFloatSection, channels [None {type=MovieSceneScriptingFloatChannel, keys [[0,1],[2,1],[10,0]]}]}]
      }
      {
       type=MovieSceneWidgetMaterialTrack
       sections [{
         type=MovieSceneParameterSection
         channels [
          Anim_Hover_Fill {type=MovieSceneScriptingFloatChannel, const [1,2]}
          Anim_Hover_Energy {type=MovieSceneScriptingFloatChannel, const [1,3]}
          Anim_Hover_Outline {type=MovieSceneScriptingFloatChannel, const [1,2]}
          Anim_OutlineThickness {type=MovieSceneScriptingFloatChannel, keys [[0,1],[4,0]]}
         ]
        }]
      }
     ]
    }]}
 }
 Anim_BackingSelect {
  binding_count=2
  length_seconds=0.2994
  object=&self:Anim_BackingSelect
  sequence {
   bindings [
    Backing {id=#3, possessed_class=Image, tracks []}
    GridSlot_1 {
     id=#4
     parent_id=#3
     possessed_class=GridSlot
     tracks [{
       type=MovieSceneMarginTrack
       display_name=Padding
       sections [{
         type=MovieSceneMarginSection
         end_frame=29
         start_frame=0
         channels [
          Left {type=MovieSceneScriptingFloatChannel, keys [[0,-2],[29.9383,0]]}
          Top {type=MovieSceneScriptingFloatChannel}
          Right {type=MovieSceneScriptingFloatChannel, keys [[0,-2],[29.9383,0]]}
          Bottom {type=MovieSceneScriptingFloatChannel}
         ]
        }]
      }]
    }
   ]
   display_rate {numerator=100}
  }
 }
 Anim_Visuals {
  binding_count=2
  length_seconds=0.7
  object=&self:Anim_Visuals
  sequence {bindings [
    VisualAttachment {
     id=#5
     possessed_class=CommonVisualAttachment
     tracks [{
       type=MovieSceneFloatTrack
       display_name=Height Override
       sections [{
         type=MovieSceneFloatSection
         end_frame=8
         start_frame=0
         channels [None {type=MovieSceneScriptingFloatChannel, keys [[0,216],[1,256],[8,216]]}]
        }]
      }]
    }
    Visuals {
     id=#6
     possessed_class=Image
     tracks [{
       type=MovieSceneWidgetMaterialTrack
       sections [{
         type=MovieSceneParameterSection
         channels [
          Flare Scale X {type=MovieSceneScriptingFloatChannel, keys [[0,0.5],[10,0.2]]}
          Flare Opacity {type=MovieSceneScriptingFloatChannel, keys [[0,1],[10,0]]}
          Triangle Remover {type=MovieSceneScriptingFloatChannel, thinned_from=4, keys [[0,1],[8,1],[14,0]]}
          Global Opacity {type=MovieSceneScriptingFloatChannel, keys [[0,0],[1,1],[14,0]]}
         ]
        }]
      }]
    }
   ]}
 }
]
graphs [
 EventGraph {nodes [
   Event OnInstantTransitionToSelected() {is=title K2Node_Event}
   > Stop All Transitions!(In:Exec) -> (Out:Exec) {is=title K2Node_MacroInstance}
   > SetRenderOpacity(self=Get FocusOutline() -> (FocusOutline:Image Object Reference), InOpacity=0)
   n1 = Get Backing() -> (Backing:Image Object Reference)
   > SetRenderOpacity(self=@n1, InOpacity=1)
   > Set Padding!(execute, Padding:Margin Structure=Make Margin(Left=0, Top=0, Right=0, Bottom=0), self:Grid Slot Object Reference=SlotAsGridSlot(Widget=@n1)) -> (then, Output_Get:Margin Structure)
   > Set Visibility(
    Visibility=HitTestInvisible
    self=Get WBP_CaptureForPostBufferUpdate() -> (WBP_CaptureForPostBufferUpdate:WBP Capture For Post Buffer Update Object Reference)
   )
   Event OnTransitionSelectedRequested() {is=title K2Node_Event}
   > Delay(Duration=0.1)
   > Branch(Condition=GetIsSelected!() -> (IsSelected:bool))
   > StopAnimation(InAnimation=Get Anim_OutlineHover() -> (Anim_OutlineHover:Widget Animation Object Reference))
   > SetRenderOpacity(self=Get FocusOutline(), InOpacity=0)
   > SetRenderOpacity(self=Get Backing(), InOpacity=1)
   > PlayAnimationForward(InAnimation=Get Anim_BackingSelect() -> (Anim_BackingSelect:Widget Animation Object Reference))
   > Set Visibility(Visibility=HitTestInvisible, self=Get WBP_CaptureForPostBufferUpdate())
   > Branch(Condition=Get UseBackgroundVisuals() -> (UseBackgroundVisuals:bool))
   > PlayAnimationForward(InAnimation=Get Anim_Visuals() -> (Anim_Visuals:Widget Animation Object Reference))
   n2 = Event OnInitialized()
   SetBackgroundMaterial!(
    execute=[@n2.then, @n7.then]
    self:WBP UIKit Block Background Base Object Reference
    Material Instance:"Material Instance Object Reference (by ref)"=Get BackgroundMaterial() -> (BackgroundMaterial:Material Instance Object Reference)
   ) -> (then)
   > SetOutlineMaterial!(
    execute
    self:WBP UIKit Block Background Base Object Reference
    Material Instance:"Material Instance Object Reference (by ref)"=Get OutlineMaterial() -> (OutlineMaterial:Material Instance Object Reference)
   ) -> (then)
   > SetCornerRadiuses!(
    execute
    self
    CornerRadius:float=Get CornerRadius() -> (CornerRadius:double)
    RadiusInPixels:bool=Get RadiusInPixels() -> (RadiusInPixels:bool)
   ) -> (then)
   > n3 = Branch(Condition=Get UseBackgroundVisuals())
   n4 = Get Visuals() -> (Visuals:Image Object Reference)
   > SetVisibility(self=@n4, InVisibility=HitTestInvisible)
   SetVisibility(execute=@n3.else, self=@n4, InVisibility=Hidden)
   Event Construct()
   > Set RenderOpacity!(execute, RenderOpacity:float=0, self:Widget Object Reference=Get FocusOutline()) -> (then, Output_Get:float)
   > Set RenderOpacity(RenderOpacity=0, self=Get Backing())
   > n5 = GetDynamicMaterial(self=Get Visuals())
   > SetScalarParameterValue(self=@n5, ParameterName=Global Opacity, Value=0)
   > Set Visibility(Visibility=Collapsed, self=Get WBP_CaptureForPostBufferUpdate())
   n6 = Event PreConstruct()
   > n7 = Branch(Condition=@n6.IsDesignTime)
   Event OnTransitionFocusedRequested() {is=title K2Node_Event}
   > n8 = Branch(Condition=GetIsSelected())
   > PlayAnimationForward(InAnimation=Get Anim_BackingSelect())
   Stop Transition!(
    In:Exec=@n8.else
    Anim:Widget Animation Object Reference=Get Anim_OutlineUnhover() -> (Anim_OutlineUnhover:Widget Animation Object Reference)
   ) -> (Out:Exec) {is=title K2Node_MacroInstance}
   > PlayAnimationForward(InAnimation=Get Anim_OutlineHover())
   Event OnTransitionUnfocusedRequested() {is=title K2Node_Event}
   > n9 = Branch(Condition=GetIsSelected())
   Stop Transition(In=@n9.else, Anim=Get Anim_OutlineHover())
   > PlayAnimationForward(InAnimation=Get Anim_OutlineUnhover())
   n10 = Event OnTransitionDeselectedFocusedRequested() {is=title K2Node_Event}
   SetRenderOpacity(execute=[@n10.then, @n11.then], self=Get Backing(), InOpacity=0)
   > StopAnimation(InAnimation=Get Anim_OutlineHover())
   > SetRenderOpacity(self=Get FocusOutline(), InOpacity=0)
   > Set Visibility(Visibility=Collapsed, self=Get WBP_CaptureForPostBufferUpdate())
   n11 = Event OnTransitionDeselectedIdleRequested() {is=title K2Node_Event}
  ]}
 SetBackgroundMaterial {nodes [
   n1 = SetBackgroundMaterial!() -> (then, Material Instance:"Material Instance Object Reference (by ref)") {is=title K2Node_FunctionEntry}
   > n2 = Set BackgroundMaterial(BackgroundMaterial:Material Instance Object Reference=@n1) -> (Output_Get:Material Instance Object Reference)
   n3 = Get Backing()
   > SetBrushFromMaterial(self=@n3, Material=@n2)
   > n4 = GetDynamicMaterial(self=@n3)
   > RefreshBackgroundMID!(execute, self:WBP UIKit Block Background Base Object Reference, Background MID:Material Instance Dynamic Object Reference=@n4) -> (then)
   > Return Node()
  ]}
 SetCornerRadiuses {nodes [
   n1 = SetCornerRadiuses!() -> (then, CornerRadius:float=~, RadiusInPixels:bool=~) {is=title K2Node_FunctionEntry}
   > Set CornerRadius(CornerRadius:double=@n1) -> (Output_Get:double)
   > Set RadiusInPixels(RadiusInPixels:bool=@n1.RadiusInPixels) -> (Output_Get:bool)
   n2 = Get OutlineDynamicMaterial() -> (OutlineDynamicMaterial:Material Instance Dynamic Object Reference)
   > SetScalarParameterValue(self=@n2, ParameterName=Radius in Pixels, Value=0)
   > SetVectorParameterValue(
    self=@n2
    ParameterName=Corner Radius
    Value=Make Linear Color!(R:float=0.3, G:float=0.3, B:float=0.3, A:float=0.3) -> (LinearColor:Linear Color Structure) {is=title K2Node_MakeStruct}
   )
   n3 = Get BackgroundDynamicMaterial() -> (BackgroundDynamicMaterial:Material Instance Dynamic Object Reference)
   > SetScalarParameterValue(self=@n3, ParameterName=Radius in Pixels, Value="To Float (Boolean)"(InBool=Get RadiusInPixels()))
   > SetScalarParameterValue(self=@n3, ParameterName=Corner Radius, Value=Get CornerRadius())
  ]}
 SetOutlineMaterial {nodes [
   n1 = SetOutlineMaterial!() -> (then, Material Instance:"Material Instance Object Reference (by ref)") {is=title K2Node_FunctionEntry}
   > n2 = Set OutlineMaterial(OutlineMaterial:Material Instance Object Reference=@n1) -> (Output_Get:Material Instance Object Reference)
   n3 = Get FocusOutline()
   > SetBrushFromMaterial(self=@n3, Material=@n2)
   > n4 = GetDynamicMaterial(self=@n3)
   > RefreshOutlineMID!(execute, self:WBP UIKit Block Background Base Object Reference, Outline MID:Material Instance Dynamic Object Reference=@n4) -> (then)
   > Return Node()
  ]}
]
variables [
 CornerRadius {type=number, default=0.325}
 RadiusInPixels {type=boolean}
 UseBackgroundVisuals {type=boolean}
]
widgets {
 info {
  inheritedWidgetCount=0
  parentClass=/FortUILibrary/Systems/ModularButton/Blocks/WBP_UIKit_BlockBackground_Base.WBP_UIKit_BlockBackground_Base_C
  rootWidgetClass=&GridPanel
  widgetCount=8
 }
 widgets [
  Root {widgetClassPath=&GridPanel, widget=Root, properties {columnFill [1], rowFill [1]}}
  VisualAttachment {
   widgetClassPath=&CommonVisualAttachment
   parent=Root
   widget=VisualAttachment
   properties {
    bOverride_HeightOverride=true
    bOverride_WidthOverride=true
    heightOverride=216
    visibility=HitTestInvisible
    widthOverride=192
    contentAnchor=<0.5 0>
   }
   slot {class=GridSlot, properties {horizontalAlignment=HAlign_Center, verticalAlignment=VAlign_Top, nudge=<0 -40>}}
  }
  Visuals {
   widgetClassPath=&Image
   bIsVariable=true
   parent=VisualAttachment
   widget=Visuals
   properties {
    visibility=Hidden
    brush {
     bIsDynamicallyLoaded=false
     drawAs=Image
     imageType=NoImage
     mirroring=NoMirror
     resourceName=None
     resourceObject=/Game/UI/Frontend/TopBar/MI_UIKit_Tab_Main_Background
     tiling=NoTile
     imageSize=<32 32>
     margin=<ltrb 0 0 0 0>
     outlineSettings {
      bUseBrushTransparency=false
      roundingType=HalfHeightRadius
      width=0
      color {colorUseRule=UseColor_Specified, specifiedColor=<rgba 0 0 0 0>}
      cornerRadii=<0 0 0 0>
     }
     tintColor {colorUseRule=UseColor_Specified, specifiedColor=<rgba 1 1 1 1>}
     uVRegion {bIsValid=false, max=<0 0>, min=<0 0>}
    }
    renderTransformPivot=<0.5 0>
   }
   slot {class=SizeBoxSlot}
  }
  SizeBox_0 {
   widgetClassPath=&SizeBox
   parent=Root
   widget=SizeBox_0
   properties {bOverride_MinDesiredHeight=true, minDesiredHeight=48}
   slot {class=GridSlot, properties {layer=1}}
  }
  Background {
   widgetClassPath=&GridPanel
   parent=SizeBox_0
   widget=Background
   properties {columnFill [1], rowFill [1]}
   slot {class=SizeBoxSlot}
  }
  WBP_CaptureForPostBufferUpdate {
   widgetClassPath=&WBP_CaptureForPostBufferUpdate
   bIsVariable=true
   parent=Background
   slot {class=GridSlot, properties {layer=-1}}
  }
  FocusOutline {
   widgetClassPath=&Image
   bIsVariable=true
   parent=Background
   widget=FocusOutline
   properties {
    pixelSnapping=Disabled
    visibility=SelfHitTestInvisible
    brush {
     bIsDynamicallyLoaded=false
     drawAs=Image
     imageType=NoImage
     mirroring=NoMirror
     resourceName=None
     resourceObject=null
     tiling=NoTile
     imageSize=<48 48>
     margin=<ltrb 0 0 0 0>
     outlineSettings {
      bUseBrushTransparency=false
      roundingType=HalfHeightRadius
      width=0
      color {colorUseRule=UseColor_Specified, specifiedColor=<rgba 0 0 0 0>}
      cornerRadii=<0 0 0 0>
     }
     tintColor {colorUseRule=UseColor_Specified, specifiedColor=<rgba 1 1 1 1>}
     uVRegion {bIsValid=false, max=<0 0>, min=<0 0>}
    }
   }
   slot {class=GridSlot, properties {padding=<ltrb -9 -9 -9 -9>}}
  }
  Backing {
   widgetClassPath=&Image
   bIsVariable=true
   parent=Background
   widget=Backing
   properties {
    brush {
     bIsDynamicallyLoaded=false
     drawAs=Image
     imageType=NoImage
     mirroring=NoMirror
     resourceName=None
     resourceObject=null
     tiling=NoTile
     imageSize=<48 48>
     margin=<ltrb 0.5 0.5 0.5 0.5>
     outlineSettings {
      bUseBrushTransparency=false
      roundingType=HalfHeightRadius
      width=0
      color {colorUseRule=UseColor_Specified, specifiedColor=<rgba 0 0 0 0>}
      cornerRadii=<0 0 0 0>
     }
     tintColor {colorUseRule=UseColor_Specified, specifiedColor=<rgba 1 1 1 1>}
     uVRegion {bIsValid=false, max=<0 0>, min=<0 0>}
    }
    colorAndOpacity=<rgba 0.7913 0.7835 0.7835 1>
   }
   slot {class=GridSlot, properties {layer=1}}
  }
 ]
}
